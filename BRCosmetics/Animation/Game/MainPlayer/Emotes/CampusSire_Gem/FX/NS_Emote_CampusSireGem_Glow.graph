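# Niagara glow effect for the CampusSire Gem emote, with AirglowLWE, Glow, Glow_Ring, Rays and Ribbon emitters, of which only Ribbon is detailed.
# After a delay, Ribbon continuously spawns short-lived points at a local offset, facing the owner's Z axis and tapering in width.
# Time-of-day functions drive color and brightness, while curves fade alpha over particle life and emitter age.
# A BlendSpace module is active; the FN ribbon setup and update modules are disabled.
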NiagaraSystem /BRCosmetics/Animation/Game/MainPlayer/Emotes/CampusSire_Gem/FX/NS_Emote_CampusSireGem_Glow
dependencies {
 usedDataInterfaces [&NiagaraDataInterfaceColorCurve, &NiagaraDataInterfaceCurve]
 usedDynamicInputs []
 usedModules [
  &self:SystemSpawnScript.SystemScriptSource.SystemScriptGraph.NiagaraNodeAssignment_0.SetVariables_#1
  &SystemState
  &BlendSpaceHelper
 ]
}
emitters {
 AirglowLWE {rendererClasses [], renderers []}
 Glow {rendererClasses [], renderers []}
 Glow_Ring {rendererClasses [], renderers []}
 Rays {rendererClasses [], renderers []}
 Ribbon {
  emitterUpdateScript [
   &EmitterState(Life Cycle Mode=::NewEnumerator1, Loop Behavior=::NewEnumerator1, Loop Delay=0.95, Loop Duration=3, UseLoopDelay=true)
   &SpawnRate(SpawnRate=80, Use Spawn Probability=false)
   null() {enabled=false, module=FN_SpawnRibbon_v2}
  ]
  particleSpawnScript [
   &InitializeParticle(
    Color Mode=::NewEnumerator0
    Lifetime=0.2
    Position Offset=<0 0 35>
    Position Offset Coordinate Space=::Local
    Ribbon Facing Mode=::NewEnumerator1
    Ribbon Facing Vector:Vector3f=$Engine.Owner.SystemZAxis
    Ribbon Width=-90
    Ribbon Width Mode=::NewEnumerator1
    UsePositionOffset=true
    Write Material Random=true
   )
   &FN_InitializeRibbon_v3(Facing Axis:ENiagaraOrientationAxis=::ZAxis, Lifetime=0.2, Ribbon Width=90) {enabled=false}
   &AddVelocity(Velocity=<0 500 400>, Velocity Speed Scale=2) {enabled=false}
  ]
  particleUpdateScript [
   &ParticleState()
   &ScaleColor(
    Scale Alpha=&Multiply_Float(
     A=&FloatFromCurve(
      FloatCurve=DataInterface() {propertyValues {Curve {keys [<at 1.5441 1 RCIM_Cubic>, {arriveTangent=-3.258, interpMode=RCIM_Cubic, leaveTangent=-3.258, tangentMode=RCTM_User, time=1.9, value=0}]}}}
      CurveIndex=$Emitter.Age
     )
     B=&FloatFromCurve(
      FloatCurve=DataInterface() {propertyValues {Curve {
         defaultValue=1
         keys [
          {arriveTangent=5.2101, interpMode=RCIM_Cubic, leaveTangent=5.2101, tangentMode=RCTM_User, time=0, value=0}
          {arriveTangent=1.9553, interpMode=RCIM_Cubic, leaveTangent=1.9553, time=0.0544, value=0.7849}
          {arriveTangent=-1.4649, interpMode=RCIM_Cubic, leaveTangent=-1.4649, tangentMode=RCTM_User, time=0.3493, value=0.683}
          {tangentMode=RCTM_User, time=1, value=0}
         ]
        }}}
      Scale Curve=&FNTimeOfDayScalar(Night Scale=0.1, Morning Scale=0.2, Day Scale=0.2, Evening Scale=0.2)
     )
    )
    Scale RGB=&MakeVectorFromLinearColorRGB(Color=&FNTimeOfDayColor(
      Night Color=<rgba 0.2309 0.2218 0.2182 1>
      Morning Color=<rgba 0.3559 0.2843 0.2563 1>
      Day Color=<rgba 0.3559 0.2843 0.2563 1>
      Evening Color=<rgba 0.3559 0.2843 0.2563 1>
     ))
    ScaleA=true
    ScaleRGB=true
   )
   &SolveForcesAndVelocity(Clamp Velocity=false, Limit Acceleration=false)
   &FN_UpdateRibbon(Enable Skeletal Position=true) {enabled=false}
   &BlendSpace(Rotation Contribution=0)
   &ScaleRibbonWidth(Ribbon Width Scale=&FloatFromCurve(FloatCurve=DataInterface() {propertyValues {Curve {
        defaultValue=1
        keys [
         {interpMode=RCIM_Cubic, tangentMode=RCTM_User, time=0, value=1}
         {arriveTangent=0.0561, interpMode=RCIM_Cubic, leaveTangent=0.0561, tangentMode=RCTM_User, time=1, value=0}
        ]
       }}}))
  ]
  rendererClasses [&NiagaraRibbonRendererProperties]
  renderers [{rendererClass=&NiagaraRibbonRendererProperties}]
 }
}
summary {
 systemName=NS_Emote_CampusSireGem_Glow
 emitters [
  {emitterName=AirglowLWE, rendererClasses []}
  {emitterName=Rays, rendererClasses []}
  {emitterName=Glow_Ring, rendererClasses []}
  {emitterName=Ribbon, rendererClasses [&NiagaraRibbonRendererProperties]}
  {emitterName=Glow, rendererClasses []}
 ]
 userVariables [
  User.RibbonIDOffset {type=NiagaraInt32}
  User.SkeletalMesh {
   type=NiagaraDataInterfaceSkeletalMesh
   defaultValue {
    struct=NiagaraExt_VariableValue_DataInterface
    value {dataInterface=&self:NiagaraDataInterfaceSkeletalMesh_0}
   }
  }
 ]
}
